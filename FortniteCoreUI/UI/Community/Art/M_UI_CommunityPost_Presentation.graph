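# Translucent UI material for community post presentation that shows an image with a gradient darkening and builds a rounded-corner mask.
# The texture parameter is sampled with scale-to-fill UVs, then multiplied by a linear gradient whose low end is lifted by a strength parameter.
# An IsVertical scalar blends between two sets of gradient start, end, rotation and strength values.
# A shared rounded-mask function takes the corner radius parameter in pixels.
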Material /FortniteCoreUI/UI/Community/Art/M_UI_CommunityPost_Presentation
graph {
 properties {
  bAutomaticallySetUsageInEditor=false
  bSubstrateRoughnessTracking=false
  bUsedWithStaticMesh=false
  blendMode=BLEND_Translucent
  materialDomain=MD_UI
  usedShadingModels=Default Lit
 }
 nodes [
  n1 = VectorParameter() {parameterName=Corner Radius, defaultValue=<rgba 32 32 32 32>}
  MaterialFunctionCall(Corner Radius=AppendVector(A=@n1.RGB, B=@n1.RGB), Radius in Pixels=Constant() {r=1}) {materialFunction=&MF_UI_CoreUI_RoundedMask}
  n2 = ScalarParameter() {defaultValue=0.05, parameterName=Gradient Strength}
  n3 = ScalarParameter() {defaultValue=0.01, parameterName=Gradient Strength Vertical}
  n4 = ScalarParameter() {defaultValue=1, parameterName=IsVertical}
  n5 = MaterialFunctionCall(
   GradientStart=LinearInterpolate(
    A=ScalarParameter() {parameterName=GradientStart}
    B=ScalarParameter() {defaultValue=0.1, parameterName=GradientStart Vertical}
    Alpha=ScalarParameter() {defaultValue=1, parameterName=IsVertical}
   )
   GradientEnd=LinearInterpolate(
    A=ScalarParameter() {defaultValue=0.5, parameterName=GradientEnd}
    B=ScalarParameter() {defaultValue=0.65, parameterName=Gradient End Vertical}
    Alpha=ScalarParameter() {defaultValue=1, parameterName=IsVertical}
   )
   RotationAngle=LinearInterpolate(
    A=ScalarParameter() {defaultValue=0.25, group=Gradient Settings, parameterName=GradientRotation, sortPriority=4}
    B=ScalarParameter() {parameterName=GradientRotation Vertical}
    Alpha=ScalarParameter() {defaultValue=1, parameterName=IsVertical}
   )
  ) {materialFunction=&MF_UI_LinearGradient}
  n6 = TextureObjectParameter() {parameterName=ImageTexture, texture=/Game/Global/Textures/Default/Blanks/T_MidgreysRGB}
  n7 = TextureProperty(Tex=@n6)
  Multiply(
   A=LinearInterpolate(A=LinearInterpolate(A=@n2, B=@n3, Alpha=@n4), B=@n5.Result, Alpha=@n5.Result) {constA=0.015, constAlpha=1}
   B=TextureSample(UVs=MaterialFunctionCall(TextureSize=@n7) {materialFunction=&MF_UI_UVScaling}.Scaled To Fill, Tex=@n6).RGB
  )
 ]
}
