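# Translucent UI material that draws an animated confetti burst, named for Community reactions.
# Two tiled grid layers randomize each cell's rotation and pick from four colors, sampling T_Confetti channels for piece shapes and spinning them over time.
# Seed, Time and Speed parameters feed named reroutes, with a power curve scaling UVs from center for the blast.
# Distance-based masks and a min(x, 1-x) envelope shape the combined confetti masks.
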Material /FortniteCoreUI/UI/Community/Art/Materials/M_Communities_Reaction_Confetti
graph {
 properties {
  bAutomaticallySetUsageInEditor=false
  bUsedWithStaticMesh=false
  blendMode=BLEND_Translucent
  materialDomain=MD_UI
  usedShadingModels=Default Lit
 }
 nodes [
  Add(A=LinearInterpolate(B=NamedRerouteUsage(), Alpha=NamedRerouteUsage()), B=LinearInterpolate(B=NamedRerouteUsage(), Alpha=NamedRerouteUsage()))
  Constant() {r=80}
  Frac(None=Time())
  n1 = Constant3Vector() {constant=<rgba 0 0.3032 1 1>}
  n2 = Constant3Vector() {constant=<rgba 1 0.2701 0 1>}
  n3 = Step(Y=NamedRerouteUsage()) {constX=0.5}
  n4 = Constant3Vector() {constant=<rgba 1 0.0066 0.2651 1>}
  n5 = Constant3Vector() {constant=<rgba 1 0.9421 0 1>}
  n6 = Step(Y=OneMinus(None=NamedRerouteUsage())) {constX=0.5}
  n7 = MaterialFunctionCall!(UV=NamedRerouteUsage(), Seed=Add(A=NamedRerouteUsage()) {constB=4}) {materialFunction=&MF_UI_Random}
  NamedRerouteDeclaration(None=LinearInterpolate(A=LinearInterpolate(A=@n1, B=@n2, Alpha=@n3), B=LinearInterpolate(A=@n4, B=@n5, Alpha=@n6), Alpha=Step(Y=@n7.Result) {constX=0.5})) {name=Confetti Colour, nodeColor=<rgba 0 0 0 1>}
  n8 = NamedRerouteUsage()
  n9 = NamedRerouteUsage()
  n10 = MaterialFunctionCall(Texture Scale=NamedRerouteDeclaration(None=Power(Base=@n8, Exp=@n9) {constExponent=0.5}) {name=Reveal, nodeColor=<rgba 0 0 0 1>}) {materialFunction=&ScaleUVsByCenter}
  n11 = Saturate(None=MaterialFunctionCall(UV=NamedRerouteDeclaration(None=@n10.UVs) {name=UV, nodeColor=<rgba 0 0 0 1>}) {materialFunction=&MF_UI_UVScaling}.Scaled to Fit)
  n12 = ScalarParameter() {defaultValue=8, parameterName=NumberOfConfetti 2}
  n13 = MaterialFunctionCall!(UV=@n11, Tiling X=@n12, Tiling Y=@n12) {materialFunction=/FortUILibrary/MaterialFunctions/Basics/MF_UI_Pattern}
  n14 = Floor(None=@n13.Raw UVs)
  n15 = MaterialFunctionCall(
   UV=NamedRerouteDeclaration(None=Divide(A=@n14, B=@n12)) {name=SteppedUV, nodeColor=<rgba 0 0 0 1>}
   Seed=Add(A=NamedRerouteUsage()) {constB=0}
  ) {materialFunction=&MF_UI_Random}
  n16 = Time()
  n17 = MaterialFunctionCall(
   UVs=Rotator(Coordinate=@n13.Pattern UVs, Time=Add(A=@n15.Result, B=@n16))
   "Rotation Angle (0-1)"=NamedRerouteDeclaration(None=@n15.Result) {name=RandomStep, nodeColor=<rgba 0.8984 0 1 1>}
  ) {materialFunction=&CustomRotator}
  n18 = Saturate(None=MaterialFunctionCall(UVs=Saturate(None=@n17.Rotated Values), Texture Scale=Constant() {r=0.95}) {materialFunction=&ScaleUVsByCenter}.UVs)
  n19 = NamedRerouteUsage()
  NamedRerouteDeclaration(None=Saturate(None=MaterialFunctionCall(UVs=@n18, Texture Scale=@n19) {materialFunction=&ScaleUVsByCenter}.UVs)) {name=Burst UV, nodeColor=<rgba 0 0 0 1>}
  n20 = Distance(A=MaterialFunctionCall() {materialFunction=&MF_UI_UVScaling}.Scaled to Fit, B=Constant() {r=0.5})
  n21 = Constant() {r=2}
  NamedRerouteDeclaration(None=OneMinus(None=Multiply(A=@n20, B=@n21) {constB=2})) {name=Mask, nodeColor=<rgba 0 0 0 1>}
  n22 = ScalarParameter() {defaultValue=10, parameterName=NumberOfConfetti 1}
  n23 = MaterialFunctionCall(UV=@n11, Tiling X=@n22, Tiling Y=@n22) {materialFunction=&MF_UI_Pattern}
  n24 = Floor(None=@n23.Raw UVs)
  n25 = MaterialFunctionCall(
   UV=NamedRerouteDeclaration(None=Divide(A=@n24, B=@n22)) {name=SteppedUV_2, nodeColor=<rgba 0 0 0 1>}
   Seed=Add(A=NamedRerouteUsage()) {constB=8}
  ) {materialFunction=&MF_UI_Random}
  n26 = Add(A=@n25.Result, B=OneMinus(None=Time()))
  n27 = MaterialFunctionCall(
   UVs=Rotator(Coordinate=@n23.Pattern UVs, Time=Multiply(A=@n26) {constB=10})
   "Rotation Angle (0-1)"=NamedRerouteDeclaration(None=@n25.Result) {name=RandomStep2, nodeColor=<rgba 0.8984 0 1 1>}
  ) {materialFunction=&CustomRotator}
  n28 = Saturate(None=MaterialFunctionCall(UVs=Saturate(None=@n27.Rotated Values), Texture Scale=Constant() {r=0.95}) {materialFunction=&ScaleUVsByCenter}.UVs)
  n29 = NamedRerouteUsage()
  NamedRerouteDeclaration(None=Saturate(None=MaterialFunctionCall(UVs=@n28, Texture Scale=@n29) {materialFunction=&ScaleUVsByCenter}.UVs)) {name=Burst UV_2, nodeColor=<rgba 0 0 0 1>}
  n30 = Constant3Vector() {constant=<rgba 1 0.2701 0 1>}
  n31 = Constant3Vector() {constant=<rgba 1 0.0066 0.2651 1>}
  n32 = Step(Y=NamedRerouteUsage()) {constX=0.5}
  n33 = Constant3Vector() {constant=<rgba 1 0.9421 0 1>}
  n34 = Constant3Vector() {constant=<rgba 0 0.3032 1 1>}
  n35 = Step(Y=NamedRerouteUsage(), X=Constant() {r=0.1}) {constX=0.5}
  n36 = MaterialFunctionCall(UV=NamedRerouteUsage(), Seed=Add(A=NamedRerouteUsage()) {constB=2}) {materialFunction=&MF_UI_Random}
  NamedRerouteDeclaration(None=LinearInterpolate(A=LinearInterpolate(A=@n30, B=@n31, Alpha=@n32), B=LinearInterpolate(A=@n33, B=@n34, Alpha=@n35), Alpha=Step(Y=@n36.Result) {constX=0.5})) {name=Confetti Colour_2, nodeColor=<rgba 0 0 0 1>}
  NamedRerouteDeclaration(None=ScalarParameter() {defaultValue=1, parameterName=Seed}) {name=Seed, nodeColor=<rgba 1 0.5156 0 1>}
  NamedRerouteDeclaration(None=ScalarParameter() {defaultValue=1, parameterName=TIme, sliderMax=1}) {name=Time, nodeColor=<rgba 0.8047 1 0 1>}
  n37 = ScalarParameter() {
   defaultValue=0.5
   desc="Controles the speed and the curve of the blast animation. (lower the value = curve of the effect) "
   parameterName=Speed
  }
  NamedRerouteDeclaration(None=@n37) {name=Speed, nodeColor=<rgba 0.875 0 1 1>}
  n38 = TextureSample(UVs=NamedRerouteUsage()) {samplerType=SAMPLERTYPE_LinearColor, texture=/FortniteCoreUI/UI/Community/Art/Textures/T_Confetti}
  n39 = NamedRerouteUsage()
  n40 = Step(Y=@n39) {constX=0.5}
  n41 = OneMinus(None=@n39)
  n42 = Distance(A=NamedRerouteUsage(), B=Constant() {r=0.5})
  n43 = LinearInterpolate(
   A=LinearInterpolate(A=LinearInterpolate(A=@n38.R, B=@n38.R, Alpha=@n40), B=@n38.G, Alpha=Step(Y=@n41) {constX=0.1})
   Alpha=Step(Y=NamedRerouteUsage(), X=Multiply(A=@n42) {constB=2}) {constX=0.33}
  ) {constB=0}
  n44 = NamedRerouteUsage()
  n45 = TextureSample(UVs=NamedRerouteUsage()) {samplerType=SAMPLERTYPE_LinearColor, texture=&T_Confetti}
  n46 = NamedRerouteUsage()
  n47 = Step(Y=@n46) {constX=0.5}
  n48 = NamedRerouteUsage()
  n49 = Distance(A=NamedRerouteUsage(), B=Constant() {r=0.5})
  n50 = LinearInterpolate(
   A=LinearInterpolate(A=LinearInterpolate(A=@n45.G, B=@n45.G, Alpha=@n47), B=@n45.B, Alpha=Step(Y=@n46) {constX=0.1})
   Alpha=Step(Y=OneMinus(None=@n48), X=Multiply(A=@n49) {constB=2}) {constX=0.33}
  ) {constB=0}
  n51 = NamedRerouteUsage()
  n52 = LinearInterpolate(
   A=NamedRerouteDeclaration(None=Multiply(A=@n43, B=@n44)) {name=Confetti Mask 1, nodeColor=<rgba 0 0 0 1>}
   Alpha=NamedRerouteDeclaration(None=Multiply(A=@n50, B=@n51)) {name=Confetti Mask 2, nodeColor=<rgba 0 0 0 1>}
  )
  n53 = NamedRerouteUsage()
  n54 = Multiply(A=Min(A=@n53, B=OneMinus(None=@n53))) {constB=4}
  NamedRerouteDeclaration(None=Saturate(None=Multiply(A=@n52, B=@n54))) {name=Combined Masks, nodeColor=<rgba 0 0 0 1>}
  NamedRerouteUsage()
 ]
}
