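# Translucent UI material for a Communities screen background, blending a two-color gradient with the Slate post-process buffer used as blur.
# An IsVertical parameter switches the linear gradient's start, end, rotation and strength between two layouts.
# UseBlur, on by default, lerps from that gradient to the GetSlatePost0 color.
# An inverted, offset radial gradient multiplied by a rounded-corner mask forms a separate shape mask.
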Material /FortniteCoreUI/UI/Community/Art/Materials/M_UI_Communities_Background
graph {
 properties {
  bAutomaticallySetUsageInEditor=false
  bUsedWithStaticMesh=false
  blendMode=BLEND_Translucent
  materialDomain=MD_UI
  usedShadingModels=Default Lit
 }
 nodes [
  n1 = ScalarParameter() {defaultValue=1, group=Gradient Settings, parameterName=IsVertical}
  n2 = MaterialFunctionCall(
   GradientStart=LinearInterpolate(
    A=ScalarParameter() {group=Gradient Settings, parameterName=GradientStart}
    B=ScalarParameter() {defaultValue=0.1, group=Gradient Settings, parameterName=GradientStart Vertical}
    Alpha=NamedRerouteDeclaration(None=@n1) {name=IsVertical, nodeColor=<rgba 0 0 0 1>}
   )
   GradientEnd=LinearInterpolate(
    A=ScalarParameter() {defaultValue=0.5, group=Gradient Settings, parameterName=GradientEnd}
    B=ScalarParameter() {defaultValue=0.65, group=Gradient Settings, parameterName=Gradient End Vertical}
    Alpha=NamedRerouteUsage()
   )
   RotationAngle=LinearInterpolate(
    A=ScalarParameter() {defaultValue=0.25, group=Gradient Settings, parameterName=GradientRotation, sortPriority=4}
    B=ScalarParameter() {group=Gradient Settings, parameterName=GradientRotation Vertical}
    Alpha=NamedRerouteUsage()
   )
  ) {materialFunction=&MF_UI_LinearGradient}
  n3 = LinearInterpolate(
   A=LinearInterpolate(
    A=ScalarParameter() {defaultValue=0.05, group=Gradient Settings, parameterName=Gradient Strength}
    B=ScalarParameter() {defaultValue=0.01, group=Gradient Settings, parameterName=Gradient Strength Vertical}
    Alpha=NamedRerouteUsage()
   )
   B=@n2.Result
   Alpha=@n2.Result
  ) {constA=0.015, constAlpha=1}
  LinearInterpolate(
   A=LinearInterpolate(
    A=VectorParameter() {group=Gradient Settings, parameterName=GradientColour2, defaultValue=<rgba 0.0168 0.0185 0.1301 1>}.RGB
    B=VectorParameter() {group=Gradient Settings, parameterName=GradientColour, defaultValue=<rgba 0.0048 0.0048 0.0052 1>}.RGB
    Alpha=OneMinus(None=@n3)
   )
   B=MaterialFunctionCall!("UVs (Optional)") {materialFunction=/Engine/Functions/UserInterface/GetSlatePost0}.LinearRGB
   Alpha=ScalarParameter() {defaultValue=1, parameterName=UseBlur}
  ) {constA=0.9}
  n4 = MaterialFunctionCall() {materialFunction=&MF_UI_UVScaling}
  n5 = Constant2Vector() {g=0.5, r=-0.5}
  n6 = MaterialFunctionCall(
   UV=MaterialFunctionCall(
    UVs=Add(A=@n4.Scaled to Fit, B=@n5) {constB=21}
    Texture Scale=ScalarParameter() {defaultValue=1.8, group=Blur Settings, parameterName=BlurGradientSize}
   ) {materialFunction=&ScaleUVsByCenter}.UVs
   GradientEnd=ScalarParameter() {defaultValue=0.75, group=Blur Settings, parameterName=BlurValue}
  ) {materialFunction=&MF_UI_RadialGradient}
  n7 = Constant() {r=10}
  n8 = Constant() {r=1}
  Max(B=Multiply(
    A=OneMinus(None=@n6.Smooth)
    B=MaterialFunctionCall(Corner Radius=@n7, Radius in Pixels=@n8) {materialFunction=&MF_UI_CoreUI_RoundedMask}.Shape Mask
   ))
 ]
}
