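# Translucent UI material for emoji reaction icons in the Communities interface.
# An Emoji Texture parameter, defaulting to the confetti emoji, lets instances swap the image fed to the texture sample.
# An IconScale parameter drives ScaleUVsByCenter, resizing the emoji around its center through its UVs.
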Material /FortniteCoreUI/UI/Community/Art/Materials/M_UI_Communities_Reaction_Emoji
graph {
 properties {
  bAutomaticallySetUsageInEditor=false
  bUsedWithStaticMesh=false
  blendMode=BLEND_Translucent
  materialDomain=MD_UI
  usedShadingModels=Default Lit
 }
 nodes [
  n1 = ScalarParameter() {defaultValue=1, parameterName=IconScale}
  TextureSample(Tex=TextureObjectParameter(UVs=MaterialFunctionCall(Texture Scale=@n1) {materialFunction=&ScaleUVsByCenter}.UVs) {parameterName=Emoji Texture, texture=/FortniteCoreUI/UI/Community/Art/Textures/T_UI_Community_Emoji_Confetti}) {texture=&T_UI_Community_Emoji_Confetti}
 ]
}
