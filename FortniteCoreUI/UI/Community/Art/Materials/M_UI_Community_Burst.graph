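# Translucent UI material that draws a glowing radial burst for Community screens, with size, reveal, segment and glow parameters.
# Scaled-to-fit UVs feed two radial-segment layers, the second rotated half a turn, each shaped by segment gradients and masks.
# SDF anti-aliasing smoothing doubles as adjustable glow on both burst layers.
# Circle SDFs build a ring pulse and a two-colour radial gradient, alongside a looping time value.
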Material /FortniteCoreUI/UI/Community/Art/Materials/M_UI_Community_Burst
graph {
 properties {
  bAutomaticallySetUsageInEditor=false
  bUsedWithStaticMesh=false
  blendMode=BLEND_Translucent
  materialDomain=MD_UI
  usedShadingModels=Default Lit
 }
 nodes [
  Frac(None=Multiply(A=Time()) {constB=0.5})
  n1 = NamedRerouteUsage()
  n2 = Constant()
  LinearInterpolate(
   A=VectorParameter() {parameterName=Colour2, defaultValue=<rgba 1 0.2464 0.0061 1>}.RGB
   B=VectorParameter() {parameterName=Colour1, defaultValue=<rgba 1 0.5925 0 1>}.RGB
   Alpha=SmoothStep(Value=MaterialFunctionCall(UVs=@n1, Size=@n2) {materialFunction=&MF_UI_Circle}.SDF) {constMax=0.65, constMin=0.4}
  )
  LinearInterpolate(A=NamedRerouteUsage(), Alpha=Multiply(A=NamedRerouteUsage()) {constB=0.5})
  n3 = Multiply(A=Distance(A=NamedRerouteUsage(), B=Constant() {r=0.5})) {constB=2}
  n4 = Constant() {r=0.5}
  n5 = Saturate(None=Multiply(A=Distance(A=@n3, B=@n4)) {constB=2})
  NamedRerouteDeclaration(None=OneMinus(None=SmoothStep(Value=@n5) {constMax=0.55})) {name=CircleMask 1, nodeColor=<rgba 0 0 0 1>}
  n6 = MaterialFunctionCall() {materialFunction=&GetUserInterfaceUV}
  NamedRerouteDeclaration(None=MaterialFunctionCall(
    UVs=MaterialFunctionCall(UV=@n6.Normalized UV) {materialFunction=&MF_UI_UVScaling}.Scaled to Fit
    Texture Scale=ScalarParameter() {defaultValue=1, group=Size, parameterName=BurstSize}
   ) {materialFunction=&ScaleUVsByCenter}.UVs) {name=UV, nodeColor=<rgba 0 0 0 1>}
  n7 = MaterialFunctionCall(UVs=NamedRerouteUsage(), Texture Scale=LinearInterpolate(Alpha=NamedRerouteUsage()) {constB=4}) {materialFunction=&ScaleUVsByCenter}
  n8 = Constant() {r=0.5}
  n9 = Multiply(A=Distance(
    A=MaterialFunctionCall(UVs=@n7.UVs, Size=@n8) {materialFunction=&MF_UI_Circle}.SDF
    B=Constant() {r=0.5}
   )) {constB=2}
  NamedRerouteDeclaration(None=Saturate(None=OneMinus(None=@n9))) {name=Pulse 1, nodeColor=<rgba 0 0 0 1>}
  NamedRerouteDeclaration(None=ScalarParameter() {defaultValue=0.1, parameterName=Reveal, sliderMax=1}) {name=Reveal, nodeColor=<rgba 0 0 0 1>}
  n10 = NamedRerouteUsage()
  n11 = NamedRerouteUsage()
  n12 = MaterialFunctionCall!(
   UV=@n11
   Segment Count
   Segment Spacing=NamedRerouteDeclaration(None=ScalarParameter() {defaultValue=0.85, parameterName=SegmentSize}) {name=SegmentGap, nodeColor=<rgba 0 0 0 1>}
   Fixed Spacing
  ) {materialFunction=/FortUILibrary/MaterialFunctions/Basics/MF_UI_RadialSegments}
  n13 = Saturate(None=@n12.Segmented Triangles Fill)
  n14 = Multiply(A=Multiply(A=Multiply(A=@n10, B=@n13), B=Saturate(None=@n12.Segmented Triangles Gradient)), B=NamedRerouteUsage())
  n15 = ScalarParameter() {defaultValue=0.09, parameterName=GlowStrength, sliderMax=0.2}
  NamedRerouteDeclaration(None=MaterialFunctionCall(
    DistanceField=Power(Base=@n14) {constExponent=0.5}
    Threshold=Constant() {r=0.1}
    Smoothing=NamedRerouteDeclaration(None=@n15) {name=Glow, nodeColor=<rgba 0 0 0 1>}
   ) {materialFunction=&MF_UI_AntiAliasing}.Outer) {name=Burst 1, nodeColor=<rgba 0 0 0 1>}
  NamedRerouteDeclaration(None=OneMinus(None=SmoothStep(Value=@n5) {constMax=0.2})) {name=CircleMask 2, nodeColor=<rgba 0 0 0 1>}
  n16 = NamedRerouteUsage()
  n17 = MaterialFunctionCall(
   UV=MaterialFunctionCall(UVs=@n11, "Rotation Angle (0-1)"=Constant() {r=0.5}) {materialFunction=&CustomRotator}.Rotated Values
   Segment Spacing=NamedRerouteUsage()
  ) {materialFunction=&MF_UI_RadialSegments}
  n18 = Saturate(None=@n17.Segmented Triangles Fill)
  n19 = Multiply(A=Multiply(A=Multiply(A=@n16, B=@n18), B=Saturate(None=@n17.Segmented Triangles Gradient)), B=NamedRerouteUsage())
  NamedRerouteDeclaration(None=MaterialFunctionCall(DistanceField=Power(Base=@n19) {constExponent=0.5}, Threshold=Constant() {r=0.1}, Smoothing=NamedRerouteUsage()) {materialFunction=&MF_UI_AntiAliasing}.Outer) {name=Burst 2, nodeColor=<rgba 0 0 0 1>}
 ]
}
