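# Translucent UI material for a community date element, drawing a rounded-corner background with a colored indicator band along one side.
# Rotated UVs, set by the Indicator Direction angle, are offset by a pixel-based size and hard-thresholded into the band mask, with the angle's absolute cosine selecting the pixel axis.
# That mask blends Side and Background colors, and combines with the rounded-mask function and a background opacity parameter to form opacity.
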Material /FortniteCoreUI/UI/Community/Art/Materials/M_UI_Community_Date
graph {
 properties {bAutomaticallySetUsageInEditor=false, blendMode=BLEND_Translucent, materialDomain=MD_UI, usedShadingModels=Default Lit}
 nodes [
  n1 = Divide(A=ScalarParameter() {defaultValue=90, desc=Angle, parameterName=Indicator Direction}) {constB=360}
  n2 = ScalarParameter() {defaultValue=20, parameterName=Indicator Size}
  n3 = Abs(None=Cosine(None=@n1))
  n4 = Add(
   A=Rotator(Coordinate=TextureCoordinate(), Time=Multiply(A=@n1) {constB=-1}) {speed=6.2832}
   B=Multiply(A=MaterialFunctionCall!(Value PX=@n2, UV Pixel Size, X or Y=@n3) {materialFunction=/FortUILibrary/MaterialFunctions/Basics/MF_UI_PixelsToNormalized}.Normalized) {constB=-1}
  )
  n5 = Saturate(None=Ceil(None=ComponentMask(None=@n4) {g=true}))
  n6 = MaterialFunctionCall(Corner Radius=ScalarParameter() {defaultValue=20, parameterName=Corner Radius}, Radius in Pixels=Constant() {r=1}) {materialFunction=&MF_UI_CoreUI_RoundedMask}
  Add(
   A=Min(A=OneMinus(None=@n5), B=@n6.Shape Mask)
   B=Multiply(A=@n6.Shape Mask, B=ScalarParameter() {defaultValue=1, parameterName=Background Opacity}) {constB=0.5}
  )
  LinearInterpolate(
   A=VectorParameter() {parameterName=Side Color, defaultValue=<rgba 1 1 1 1>}.RGB
   B=VectorParameter() {parameterName=Background Color, defaultValue=<rgba 0.2 0.2 0.2 1>}.RGB
   Alpha=@n5
  ) {constAlpha=0.15}
 ]
}
