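# UI material for the community footer backplate, a rounded panel that darkens what is behind it through Modulate blending.
# A vertical gradient runs from a fixed light shade at the top to the GradientOpacity value at the bottom, faded toward white by vertex alpha.
# MF_UI_CoreUI_RoundedMask, driven by per-corner radius parameters, limits the effect to the shape, outputting white outside so the background stays unchanged.
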Material /FortniteCoreUI/UI/Community/Art/Materials/M_UI_Community_FooterBackplate
graph {
 properties {
  bAutomaticallySetUsageInEditor=false
  bUsedWithStaticMesh=false
  blendMode=BLEND_Modulate
  materialDomain=MD_UI
  usedShadingModels=Default Lit
 }
 nodes [
  n1 = ScalarParameter() {defaultValue=0.4, group=Opacity, parameterName=GradientOpacity}
  n2 = ComponentMask(None=MaterialFunctionCall() {materialFunction=&GetUserInterfaceUV}.Normalized UV) {g=true}
  n3 = VertexColor()
  n4 = VectorParameter() {
   group=Shape
   parameterName=Corner Radius
   sortPriority=1
   channelNames {a=BottomRight, b=TopRight, g=BottomLeft, r=TopLeft}
   defaultValue=<rgba 16 16 16 16>
  }
  LinearInterpolate(
   B=LinearInterpolate(B=LinearInterpolate(B=@n1, Alpha=@n2) {constA=0.8, constB=0.4}, Alpha=Multiply(A=@n3.A)) {constA=1, constB=0}
   Alpha=MaterialFunctionCall(
    Corner Radius=AppendVector(A=@n4, B=@n4)
    Radius in Pixels=ScalarParameter() {defaultValue=1, group=Shape, parameterName=Radius in Pixels, sliderMax=1, sortPriority=2}
   ) {materialFunction=&MF_UI_CoreUI_RoundedMask}.Shape Mask
  ) {constA=1}
 ]
}
