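# UI material for a translucent rounded-rectangle backplate, named for the QR code display in the Community UI.
# The shared CoreUI rounded-mask function builds the shape from per-corner radius parameters and a radius-in-pixels setting.
# That mask is multiplied by an Opacity scalar, while a separate BackplateColour parameter provides the colour, white by default.
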Material /FortniteCoreUI/UI/Community/Art/Materials/M_UI_Community_QRBackplate
graph {
 properties {
  bAutomaticallySetUsageInEditor=false
  bUsedWithStaticMesh=false
  blendMode=BLEND_Translucent
  materialDomain=MD_UI
  usedShadingModels=Default Lit
 }
 nodes [
  n1 = VectorParameter() {
   group=Shape
   parameterName=Corner Radius
   sortPriority=1
   channelNames {a=BottomRight, b=TopRight, g=BottomLeft, r=TopLeft}
   defaultValue=<rgba 16 16 16 16>
  }
  Multiply(
   A=MaterialFunctionCall(
    Corner Radius=AppendVector(A=@n1, B=@n1)
    Radius in Pixels=ScalarParameter() {defaultValue=1, group=Shape, parameterName=Radius in Pixels, sliderMax=1, sortPriority=2}
   ) {materialFunction=&MF_UI_CoreUI_RoundedMask}.Shape Mask
   B=ScalarParameter() {defaultValue=1, group=Colour, parameterName=Opacity}
  )
  VectorParameter() {group=Colour, parameterName=BackplateColour, defaultValue=<rgba 1 1 1 1>}
 ]
}
