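# Translucent UI material for a community reaction backplate: a rounded rectangle with warm orange glows and an outline.
# Two offset, stretched radial gradients from aspect-scaled UVs form the glows, blended at reduced strength and colored through multi-stop orange-to-gold lerps.
# A rounded-mask function with per-corner radius parameters supplies background and outline masks.
# A mirrored horizontal mask, rising away from center and scaled by OutlineOpacity, is also built.
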Material /FortniteCoreUI/UI/Community/Art/Materials/M_UI_Community_Reaction_Backplate
graph {
 properties {
  bAutomaticallySetUsageInEditor=false
  bUsedWithStaticMesh=false
  blendMode=BLEND_Translucent
  materialDomain=MD_UI
  usedShadingModels=Default Lit
 }
 nodes [
  n1 = MaterialFunctionCall(UV=NamedRerouteDeclaration(None=MaterialFunctionCall() {materialFunction=&GetUserInterfaceUV}.Normalized UV) {name=UVs, nodeColor=<rgba 0 0 0 1>}) {materialFunction=&MF_UI_UVScaling}
  n2 = Add(A=@n1.Scaled to Fit, B=Constant2Vector() {g=-0.8, r=2.5})
  n3 = Constant2Vector() {g=3.2, r=10}
  n4 = Power(Base=MaterialFunctionCall(UV=MaterialFunctionCall(UVs=@n2, Texture Scale=@n3) {materialFunction=&ScaleUVsByCenter}.UVs) {materialFunction=&MF_UI_RadialGradient}.Smooth)
  n5 = Add(A=@n1.Scaled to Fit, B=Constant2Vector() {g=-0.8, r=-2.5})
  n6 = Constant2Vector() {g=3.2, r=10}
  n7 = Power(Base=MaterialFunctionCall(UV=MaterialFunctionCall(UVs=@n5, Texture Scale=@n6) {materialFunction=&ScaleUVsByCenter}.UVs) {materialFunction=&MF_UI_RadialGradient}.Smooth)
  n8 = LinearInterpolate(
   A=Multiply(A=NamedRerouteDeclaration(None=@n4) {name=Glow1, nodeColor=<rgba 1 0 0.8437 1>}) {constB=0.4}
   Alpha=Multiply(A=NamedRerouteDeclaration(None=@n7) {name=glow2, nodeColor=<rgba 0 1 0.8359 1>}) {constB=0.4}
  )
  n9 = NamedRerouteUsage()
  n10 = NamedRerouteUsage()
  n11 = Constant() {r=0.15}
  LinearInterpolate(A=LinearInterpolate(A=LinearInterpolate(B=@n8, Alpha=@n9), Alpha=Multiply(A=@n10, B=@n11)), B=NamedRerouteUsage(), Alpha=NamedRerouteUsage())
  n12 = MaterialFunctionCall(
   1=Constant3Vector() {constant=<rgba 1 0.0999 0 1>}
   2=Constant3Vector() {constant=<rgba 1 0.3619 0.0248 1>}
   3=Constant3Vector() {constant=<rgba 0.7991 0.4564 0.0908 1>}
   A=NamedRerouteUsage()
  ) {materialFunction=&Lerp_Multiple_Float3}
  n13 = NamedRerouteUsage()
  n14 = MaterialFunctionCall(
   1=Constant3Vector() {constant=<rgba 1 0.0999 0 1>}
   2=Constant3Vector() {constant=<rgba 1 0.3619 0.0248 1>}
   3=Constant3Vector() {constant=<rgba 0.7991 0.4564 0.0908 1>}
   A=NamedRerouteUsage()
  ) {materialFunction=&Lerp_Multiple_Float3}
  n15 = NamedRerouteUsage()
  LinearInterpolate(
   A=LinearInterpolate(A=LinearInterpolate(B=@n12.Lerp 3 Inputs, Alpha=@n13), Alpha=LinearInterpolate(B=@n14.Lerp 3 Inputs, Alpha=@n15))
   B=Constant3Vector() {constant=<rgba 1 0.1119 0 1>}
   Alpha=NamedRerouteUsage()
  )
  n16 = VectorParameter() {
   group=Shape
   parameterName=Corner Radius
   sortPriority=1
   channelNames {a=BottomRight, b=TopRight, g=BottomLeft, r=TopLeft}
   defaultValue=<rgba 16 16 16 16>
  }
  n17 = MaterialFunctionCall(
   Corner Radius=AppendVector(A=@n16, B=@n16)
   Radius in Pixels=ScalarParameter() {defaultValue=1, group=Shape, parameterName=Radius in Pixels, sliderMax=1, sortPriority=2}
   Outline Thickness=Constant() {r=2}
  ) {materialFunction=&MF_UI_CoreUI_RoundedMask}
  NamedRerouteDeclaration(None=@n17.Outline Mask) {name=Outline Mask, nodeColor=<rgba 0.0781 0 1 1>}
  NamedRerouteDeclaration(None=@n17.Shape Mask) {name=BackgroundMask, nodeColor=<rgba 0 1 0.2969 1>}
  n18 = NamedRerouteUsage()
  n19 = Multiply(A=Distance(A=ComponentMask(None=@n18) {r=true}, B=Constant() {r=0.5})) {constB=4}
  NamedRerouteDeclaration(None=Multiply(A=Saturate(None=@n19), B=ScalarParameter() {defaultValue=0.6, parameterName=OutlineOpacity}) {constB=0.6}) {name=MirrorMask, nodeColor=<rgba 0 0 0 1>}
 ]
}
